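# Physics asset for the collar of the female medium Cyber Karate body cosmetic, setting up two dynamic back collar flaps, left and right.
# Each box-shaped flap is constrained to a small master body with one swing locked and twist nearly locked, leaving a hinge-like limited swing.
# Both masters and the capsule clavicle bodies connect to the spine_05 box through angularly free constraints, with clavicles and spine providing upper-torso shapes.
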PhysicsAsset /BRCosmetics/Characters/Player/Female/Medium/Bodies/F_MED_Cyber_Karate/Meshes/F_MED_Cyber_Karate_Collar_Physics
bodies {
 clavicle_l [{shapeType=Capsule, length=11.25, radius=6.5, center=<8.4239 0.2055 -3.0357>, rotation=<rot 90 179.1048 176.3563>}]
 clavicle_r [{shapeType=Capsule, length=11.25, radius=6.5, center=<-8.2296 -0.1716 3.0354>, rotation=<rot 90 -10.1866 -12.3287>}]
 dyn_collar_back_l [{shapeType=Box, extentX=13.323, extentY=1.6, extentZ=12.9826, center=<2.0149 -0.3286 0>, rotation=<rot 0 -7.7317 0>}]
 dyn_collar_back_r [{shapeType=Box, extentX=13.323, extentY=1.6, extentZ=13.1837, center=<2.0149 -0.3286 0>, rotation=<rot 0 -7.7317 0>}]
 master_collar_back_l [{shapeType=Box, extentX=1.01, extentY=1.01, extentZ=1.01, rotation=<rot 0 -90 0>}]
 master_collar_back_r [{shapeType=Box, extentX=1.01, extentY=1.01, extentZ=1.01, rotation=<rot 0 -90 0>}]
 spine_05 [{
   shapeType=Box
   extentX=2.0039
   extentY=24.9632
   extentZ=16.4077
   shapeName=spine_05_box
   center=<-2.4551 -6.9095 -0.0014>
   rotation=<rot -0.2929 98.4365 -89.8373>
  }]
}
constraints [
 {
  bone1Name=dyn_collar_back_l
  bone2Name=master_collar_back_l
  swing1LimitDegrees=50
  swing2LimitDegrees=45
  swing2Motion=Locked
  twistLimitDegrees=1
 }
 {
  bone1Name=dyn_collar_back_r
  bone2Name=master_collar_back_r
  swing1LimitDegrees=50
  swing2LimitDegrees=45
  swing2Motion=Locked
  twistLimitDegrees=1
 }
 {
  bone1Name=clavicle_l
  bone2Name=spine_05
  swing1LimitDegrees=45
  swing1Motion=Free
  swing2LimitDegrees=45
  swing2Motion=Free
  twistLimitDegrees=45
  twistMotion=Free
 }
 {
  bone1Name=master_collar_back_l
  bone2Name=spine_05
  swing1LimitDegrees=45
  swing1Motion=Free
  swing2LimitDegrees=45
  swing2Motion=Free
  twistLimitDegrees=45
  twistMotion=Free
 }
 {
  bone1Name=clavicle_r
  bone2Name=spine_05
  swing1LimitDegrees=45
  swing1Motion=Free
  swing2LimitDegrees=45
  swing2Motion=Free
  twistLimitDegrees=45
  twistMotion=Free
 }
 {
  bone1Name=master_collar_back_r
  bone2Name=spine_05
  swing1LimitDegrees=45
  swing1Motion=Free
  swing2LimitDegrees=45
  swing2Motion=Free
  twistLimitDegrees=45
  twistMotion=Free
 }
]
