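# Physics asset for the scarf on the female medium Cyber Karate cosmetic body, defining its dynamic tie bones and the body shapes around them.
# Four boxes form a chain from dyn_tie_01 to dyn_tie_04, hung from a spine_05 box with limited swing and very tight twist.
# Capsules on spine_03, both forearms and both thighs give body collision volumes; spine_05 joins spine_03 with free angular motion.
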PhysicsAsset /BRCosmetics/Characters/Player/Female/Medium/Bodies/F_MED_Cyber_Karate/Meshes/F_MED_Cyber_Karate_Scarf_Physics
bodies {
 dyn_tie_01 [{shapeType=Box, extentX=3.5, extentY=3.5, extentZ=1.5, center=<0.7624 0 0.1828>, rotation=<rot 0 -89.9999 6.6161>}]
 dyn_tie_02 [{shapeType=Box, extentX=4.5, extentY=3.5, extentZ=2, center=<1.1228 -0.0001 0>, rotation=<rot 0 -90 0>}]
 dyn_tie_03 [{shapeType=Box, extentX=6.9529, extentY=2.3916, extentZ=8.6362, center=<1.6414 0 0>, rotation=<rot 0.1294 -179.226 -90.4248>}]
 dyn_tie_04 [{shapeType=Box, extentX=7.1959, extentY=2.2682, extentZ=9.2814, center=<3.0772 -0.0186 0.0575>, rotation=<rot -2.6997 178.0075 -88.7387>}]
 lowerarm_l [{
   shapeType=Capsule
   length=21.75
   radius=3.5024
   shapeName=lowerarm_l_capsule
   center=<14.3461 0 -0.7349>
   rotation=<rot 87.0676 0.0005 0.0004>
  }]
 lowerarm_r [{
   shapeType=Capsule
   length=21.75
   radius=3.5024
   shapeName=lowerarm_r_capsule
   center=<-14.4816 0 0.7418>
   rotation=<rot 87.0676 0.0006 0.0006>
  }]
 spine_03 [
  {shapeType=Capsule, length=10.039, radius=11.918, center=<-7.2014 0.3423 -0.518>, rotation=<rot -2.8498 -84.6116 86.722>}
  {shapeType=Capsule, length=14.75, radius=6.75, center=<5.9545 5.6792 -0.0001>}
 ]
 spine_05 [{shapeType=Box, extentX=3.1404, extentY=3.7958, extentZ=3.802, center=<1.2778 0.5393 -0.0692>, rotation=<rot 0 41.3674 -89.5616>}]
 thigh_l [{
   shapeType=Capsule
   length=25.5
   radius=13.5
   shapeName=thigh_l_capsule
   center=<-13.1122 -0.0003 -0.542>
   rotation=<rot -87.633 0.0005 -0.0005>
  }]
 thigh_r [{
   shapeType=Capsule
   length=25.5
   radius=13.5
   shapeName=thigh_r_capsule
   center=<12.0648 -0.0008 0.4992>
   rotation=<rot -87.6306 0.0005 -0.0005>
  }]
}
constraints [
 {bone1Name=dyn_tie_02, bone2Name=dyn_tie_01, swing1LimitDegrees=20, swing2LimitDegrees=40, twistLimitDegrees=3}
 {bone1Name=dyn_tie_01, bone2Name=spine_05, swing1LimitDegrees=20, swing2LimitDegrees=45, twistLimitDegrees=3}
 {bone1Name=dyn_tie_03, bone2Name=dyn_tie_02, swing1LimitDegrees=25, swing2LimitDegrees=40, twistLimitDegrees=3}
 {bone1Name=dyn_tie_04, bone2Name=dyn_tie_03, swing1LimitDegrees=35, swing2LimitDegrees=30, twistLimitDegrees=3}
 {
  bone1Name=spine_05
  bone2Name=spine_03
  swing1LimitDegrees=45
  swing1Motion=Free
  swing2LimitDegrees=45
  swing2Motion=Free
  twistLimitDegrees=45
  twistMotion=Free
 }
]
