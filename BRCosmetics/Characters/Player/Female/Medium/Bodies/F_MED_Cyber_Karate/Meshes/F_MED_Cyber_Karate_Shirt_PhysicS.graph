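# Physics asset for the shirt of the female medium Cyber Karate player body, setting up its dynamic shirt and armband bones for physics.
# Six box bodies on left and right shirt bones each constrain directly to a spine_03 capsule with tight swing and twist limits.
# The armband's master knot hangs from spine_03 with looser limits and carries the knot, which branches into two-link high and low capsule chains.
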PhysicsAsset /BRCosmetics/Characters/Player/Female/Medium/Bodies/F_MED_Cyber_Karate/Meshes/F_MED_Cyber_Karate_Shirt_PhysicS
bodies {
 dyn_armband_hi_01 [{shapeType=Capsule, length=1.5, radius=1, center=<1.3052 0.3037 0>, rotation=<rot 86.1526 -171.4561 179.9999>}]
 dyn_armband_hi_02 [{shapeType=Capsule, length=1, radius=1, center=<0.5414 -0.0261 -0.1905>, rotation=<rot 90 0 0>}]
 dyn_armband_knot [{shapeType=Capsule, length=1, radius=1, rotation=<rot 90 0 0>}]
 dyn_armband_knot_MSTR [{shapeType=Capsule, length=0.505, radius=0.505, rotation=<rot 0 -90 0>}]
 dyn_armband_low_01 [{shapeType=Capsule, length=2, radius=1, center=<2 -2 0>, rotation=<rot 70 0 0>}]
 dyn_armband_low_02 [{shapeType=Capsule, length=2.5, radius=1, center=<0.7577 0.0001 -0.0886>, rotation=<rot 83.3318 -0.0002 -0.0001>}]
 dyn_shirt_01_l [{shapeType=Box, extentX=1.5, extentY=6.9962, extentZ=4.379, center=<1.4209 -0.2062 -0.1043>, rotation=<rot -34.1106 -89.2327 -91.5379>}]
 dyn_shirt_01_r [{shapeType=Box, extentX=1.2722, extentY=3.5073, extentZ=5.0231, center=<0.8725 0.0014 0.0064>, rotation=<rot 11.9974 -89.9995 -0.4308>}]
 dyn_shirt_02_l [{shapeType=Box, extentX=4.1297, extentY=1.205, extentZ=8.4735, center=<1.049 0.013 -0.0395>, rotation=<rot 2.156 -179.2908 -65.6467>}]
 dyn_shirt_02_r [{shapeType=Box, extentX=5.6623, extentY=1.2307, extentZ=4.1408, center=<1.5333 -0.074 0.0078>, rotation=<rot 16.685 89.7228 -90.1131>}]
 dyn_shirt_03_l [{shapeType=Box, extentX=1.01, extentY=4.01, extentZ=5.51, center=<1.2928 0.0017 0.0165>, rotation=<rot 6.0451 -89.9995 0.9857>}]
 dyn_shirt_03_r [{shapeType=Box, extentX=4.3353, extentY=1.5, extentZ=6.2848, center=<1.1357 0.0403 0.0639>, rotation=<rot -1.3187 -1.9748 -46.2903>}]
 spine_03 [{shapeType=Capsule, length=12, radius=7.5, center=<-2.3391 2.2902 0.0207>, rotation=<rot -89.4561 -179.999 -171.7289>}]
}
constraints [
 {bone1Name=dyn_shirt_01_l, bone2Name=spine_03, swing1LimitDegrees=20, swing2LimitDegrees=20, twistLimitDegrees=3}
 {bone1Name=dyn_shirt_02_r, bone2Name=spine_03, swing1LimitDegrees=20, swing2LimitDegrees=20, twistLimitDegrees=3}
 {bone1Name=dyn_shirt_03_r, bone2Name=spine_03, swing1LimitDegrees=20, swing2LimitDegrees=15, twistLimitDegrees=3}
 {bone1Name=dyn_shirt_01_r, bone2Name=spine_03, swing1LimitDegrees=20, swing2LimitDegrees=20, twistLimitDegrees=3}
 {bone1Name=dyn_shirt_02_l, bone2Name=spine_03, swing1LimitDegrees=20, swing2LimitDegrees=20, twistLimitDegrees=3}
 {bone1Name=dyn_shirt_03_l, bone2Name=spine_03, swing1LimitDegrees=20, swing2LimitDegrees=10, twistLimitDegrees=2}
 {bone1Name=dyn_armband_knot_MSTR, bone2Name=spine_03, swing1LimitDegrees=45, swing2LimitDegrees=45, twistLimitDegrees=45}
 {bone1Name=dyn_armband_knot, bone2Name=dyn_armband_knot_MSTR, swing1LimitDegrees=10, swing2LimitDegrees=10, twistLimitDegrees=1}
 {bone1Name=dyn_armband_low_01, bone2Name=dyn_armband_knot, swing1LimitDegrees=30, swing2LimitDegrees=30, twistLimitDegrees=3}
 {bone1Name=dyn_armband_low_02, bone2Name=dyn_armband_low_01, swing1LimitDegrees=30, swing2LimitDegrees=30, twistLimitDegrees=3}
 {bone1Name=dyn_armband_hi_01, bone2Name=dyn_armband_knot, swing1LimitDegrees=25, swing2LimitDegrees=25, twistLimitDegrees=3}
 {bone1Name=dyn_armband_hi_02, bone2Name=dyn_armband_hi_01, swing1LimitDegrees=30, swing2LimitDegrees=30, twistLimitDegrees=5}
]
