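# Post-process style Anim Blueprint for the F_MED_Cyber_Karate Fortnite body, adding secondary motion to cloth and accessory bones over the parent mesh's copied pose.
# RigidBody simulation or constraint fallback, chosen by LOD and physics state, feeds further constraints plus collar and shirt simulations.
# The cached result gets crouch-dependent flap and collar offsets, then a Control Rig.
# The event graph sets crouch physics blending, jump force and backpack-aware collar disabling.
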AnimBlueprint /BRCosmetics/Characters/Player/Female/Medium/Bodies/F_MED_Cyber_Karate/Meshes/F_MED_Cyber_Karate_AnimBp
target_skeleton=&/F_MED_Cyber_Karate_Skeleton
graphs [
 AnimGraph {nodes [
   n1 = "Transform (Modify) Bone - Bone: Flap_Front_Driven_Bone"(
    ComponentPose=Local To Component(LocalPose=Use cached pose 'PreCrouch'())
    Alpha=1
    Translation="0,0,0"
    Rotation="32.680,26.356,-52.2"
    Scale="1,1,1"
   ) {
    rotationMode=BMM_Additive
    rotationSpace=BCS_BoneSpace
    translationMode=BMM_Additive
    translationSpace=BCS_BoneSpace
    alphaBoolBlend {blendOption=Linear}
    alphaScaleBiasClamp {bInterpResult=true}
    boneToModify {boneName=Flap_Front_Driven_Bone}
    rotation=<rot 32.68 26.356 -52.2>
   }
   n2 = Component To Local(ComponentPose=@n1)
   n3 = "Transform (Modify) Bone - Bone: driver_frill_frt_l"(
    ComponentPose=Local To Component(LocalPose=Use cached pose 'PreCrouch'())
    Alpha=1
    Translation="0,0,0"
    Rotation="0.000,-15,0.000"
    Scale="1,1,1"
   ) {
    rotationMode=BMM_Additive
    rotationSpace=BCS_BoneSpace
    alphaBoolBlend {blendOption=Linear}
    boneToModify {boneName=driver_frill_frt_l}
    rotation {yaw=-15}
   }
   n4 = "Transform (Modify) Bone - Bone: driver_frill_frt_r"(ComponentPose=@n3, Alpha=1, Translation="0,0,0", Rotation="0.000,-24.999,0.000", Scale="1,1,1") {
    rotationMode=BMM_Additive
    rotationSpace=BCS_BoneSpace
    alphaBoolBlend {blendOption=Linear}
    boneToModify {boneName=driver_frill_frt_r}
    rotation {yaw=-24.9993}
   }
   n5 = "Transform (Modify) Bone - Bone: Flap_Front_Parent_Space"(ComponentPose=@n4, Alpha=1, Translation="0,0,0", Rotation="0.000,0,-55", Scale="1,1,1") {
    rotationMode=BMM_Additive
    rotationSpace=BCS_BoneSpace
    alphaBoolBlend {blendOption=Linear}
    boneToModify {boneName=Flap_Front_Parent_Space}
    rotation {roll=-55}
   }
   n6 = Component To Local(ComponentPose=@n5)
   n7 = Get bIsCrouching()
   n8 = "Transform (Modify) Bone - Bone: master_collar_back_l"(
    ComponentPose=Local To Component(LocalPose=Blend Poses by bool(BlendPose_0=@n2, BlendPose_1=@n6, bActiveValue=@n7) {blendType=Linear})
    Alpha=1
    Translation="-0.5,-1,0"
    Rotation="0,5,0"
    Scale="1,1,1"
   ) {
    rotationMode=BMM_Additive
    rotationSpace=BCS_BoneSpace
    scaleMode=BMM_Additive
    scaleSpace=BCS_BoneSpace
    translationMode=BMM_Additive
    translationSpace=BCS_BoneSpace
    alphaBoolBlend {blendOption=Linear}
    boneToModify {boneName=master_collar_back_l}
    rotation {yaw=5}
    translation=<-0.5 -1>
   }
   n9 = "Transform (Modify) Bone - Bone: master_collar_back_r"(ComponentPose=@n8, Alpha=1, Translation="-0.5,-1,0", Rotation="0,5,0", Scale="1,1,1") {
    rotationMode=BMM_Additive
    rotationSpace=BCS_BoneSpace
    scaleMode=BMM_Additive
    scaleSpace=BCS_BoneSpace
    translationMode=BMM_Additive
    translationSpace=BCS_BoneSpace
    alphaBoolBlend {blendOption=Linear}
    boneToModify {boneName=master_collar_back_r}
    rotation {yaw=5}
    translation=<-0.5 -1>
   }
   n10 = Control Rig!(
    Source=Component To Local(ComponentPose=@n9)
    Alpha
    On emotes:bool=Get bIsPlayingEmote()
    BackPack On:bool=Get isBackStatic() -> (isBackStatic:bool)
    Is in Front End:bool=Get bIsInFrontEnd()
   ) -> (Pose) {
    alphaBoolBlend {blendOption=Linear}
    controlRigAssetReference {blueprintRigClass=&/F_MED_Cyber_Karate_CtrlRig.F_MED_Cyber_Karate_CtrlRig_C}
   }
   Output Pose(Result=@n10) {layerGroup=None}
   n11 = RigidBody(
    ComponentPose=Local To Component(LocalPose=Copy Pose From Mesh() {bUseAttachedParent=true})
    Alpha=Get crouch_rigid_alpha() -> (crouch_rigid_alpha:double)
   ) {
    bOverrideWorldGravity=true
    overridePhysicsAsset=&/F_MED_Cyber_Karate_Physics
    simulationSpace=BaseBoneSpace
    alphaBoolBlend {blendOption=Linear}
    baseBoneRef {boneName=root}
    componentAppliedLinearAccClamp=<1000 1000 1000>
    componentLinearAccScale=<1 1 1>
    componentLinearVelScale=<1 1 1>
    overrideWorldGravity {z=-980}
   }
   n12 = RigidBody(ComponentPose=@n11, Alpha=0.7) {
    bClampLinearTranslationLimitToRefPose=true
    bOverrideWorldGravity=true
    overridePhysicsAsset=&/F_MED_Cyber_Karate_Scarf_Physics
    simulationSpace=BaseBoneSpace
    alphaBoolBlend {blendOption=Linear}
    baseBoneRef {boneName=root}
    componentAppliedLinearAccClamp=<6000 6000 6000>
    componentLinearAccScale=<0.5 0.5 0.5>
    componentLinearVelScale=<0.5 0.5 0.5>
    overrideWorldGravity {z=-980}
   }
   n13 = "Transform (Modify) Bone - Bone: dyn_tie_02"(ComponentPose=@n12, Alpha=1, Translation="0,0,0", Rotation="-5,0,0", Scale="1,1,1") {
    rotationMode=BMM_Additive
    rotationSpace=BCS_ParentBoneSpace
    alphaBoolBlend {blendOption=Linear}
    boneToModify {boneName=dyn_tie_02}
    rotation {pitch=-5}
   }
   n14 = Component To Local(ComponentPose=@n13)
   n15 = "Constraint - dyn_tie_02 (P)"(ComponentPose=Local To Component(LocalPose=Copy Pose From Mesh() {bUseAttachedParent=true}), Alpha=1, ConstraintWeights_0=1) {
    alphaBoolBlend {blendOption=Linear}
    boneToModify {boneName=dyn_tie_02}
    constraintSetup [{targetBone {boneName=spine_04}}]
    constraintWeights [1]
   }
   n16 = "Constraint - dyn_tie_03 (PP)"!(ComponentPose=@n15, Alpha=1, ConstraintWeights_0=0.25, ConstraintWeights_1:float=0.75) -> (Pose) {
    alphaBoolBlend {blendOption=Linear}
    boneToModify {boneName=dyn_tie_03}
    constraintSetup [{targetBone {boneName=spine_04}}, {targetBone {boneName=spine_03}}]
    constraintWeights [1,1]
   }
   n17 = "Constraint - dyn_tie_04 (PP)"!(ComponentPose=@n16, Alpha=1, ConstraintWeights_0=0.3, ConstraintWeights_1:float=0.7) -> (Pose) {
    alphaBoolBlend {blendOption=Linear}
    boneToModify {boneName=dyn_tie_04}
    constraintSetup [{targetBone {boneName=spine_03}}, {targetBone {boneName=spine_02}}]
    constraintWeights [1,1]
   }
   n18 = "Constraint - dyn_tie_05 (PP)"!(ComponentPose=@n17, Alpha=1, ConstraintWeights_0=0.2, ConstraintWeights_1:float=0.8) -> (Pose) {
    alphaBoolBlend {blendOption=Linear}
    boneToModify {boneName=dyn_tie_05}
    constraintSetup [{targetBone {boneName=spine_02}}, {targetBone {boneName=spine_01}}]
    constraintWeights [1,1]
   }
   n19 = "Transform (Modify) Bone - Bone: dyn_tie_01"(ComponentPose=@n18, Alpha=1, Translation="0,0,0", Rotation="0.000,15,0.000", Scale="1,1,1") {
    rotationMode=BMM_Additive
    rotationSpace=BCS_ParentBoneSpace
    alphaBoolBlend {blendOption=Linear}
    boneToModify {boneName=dyn_tie_01}
    rotation {yaw=15}
   }
   n20 = "Transform (Modify) Bone - Bone: dyn_tie_02"(ComponentPose=@n19, Alpha=1, Translation="0,0,0", Rotation="-15,0.000,-0.000", Scale="1,1,1") {
    rotationMode=BMM_Additive
    rotationSpace=BCS_ParentBoneSpace
    alphaBoolBlend {blendOption=Linear}
    boneToModify {boneName=dyn_tie_02}
    rotation {pitch=-15}
   }
   n21 = Component To Local(ComponentPose=@n20)
   n22 = Get isRigidBodyDisabled() -> (isRigidBodyDisabled:bool)
   n23 = "Constraint - const_pouch_r (PP)"!(
    ComponentPose=Local To Component(LocalPose=Blend Poses by bool(BlendPose_0=@n21, BlendPose_1=@n14, bActiveValue=@n22) {blendType=Linear})
    Alpha=1
    ConstraintWeights_0=0.885
    ConstraintWeights_1:float=0.115
   ) -> (Pose) {
    alphaBoolBlend {blendOption=Linear}
    boneToModify {boneName=const_pouch_r}
    constraintSetup [{targetBone {boneName=thigh_r}}, {targetBone {boneName=thigh_twist_01_r}}]
    constraintWeights [1,1]
   }
   n24 = "Constraint - dyn_frill_frt (PPPPPPP)"!(
    ComponentPose=@n23
    Alpha=1
    ConstraintWeights_0=0.274
    ConstraintWeights_1:float=0.1273
    ConstraintWeights_2:float=0.1438
    ConstraintWeights_3:float=0.2665
    ConstraintWeights_4:float=0.0178
    ConstraintWeights_5:float=0.0679
    ConstraintWeights_6:float=0.1027
   ) -> (Pose) {
    alphaBoolBlend {blendOption=Linear}
    boneToModify {boneName=dyn_frill_frt}
    constraintSetup [
     {targetBone {boneName=thigh_r}}
     {targetBone {boneName=thigh_twist_01_r}}
     {targetBone {boneName=thigh_twist_01_l}}
     {targetBone {boneName=thigh_l}}
     {targetBone {boneName=dyn_frill_frt}}
     {targetBone {boneName=pelvis}}
     {targetBone {boneName=dyn_frill_side_r}}
    ]
    constraintWeights [1,1,1,1,1,1,1]
   }
   n25 = "Constraint - dyn_frill_bck (PPPPP)"!(
    ComponentPose=@n24
    Alpha=1
    ConstraintWeights_0=0.0484
    ConstraintWeights_1:float=0.4698
    ConstraintWeights_2:float=0.2345
    ConstraintWeights_3:float=0.2005
    ConstraintWeights_4:float=0.0468
   ) -> (Pose) {
    alphaBoolBlend {blendOption=Linear}
    boneToModify {boneName=dyn_frill_bck}
    constraintSetup [
     {targetBone {boneName=thigh_r}}
     {targetBone {boneName=pelvis}}
     {targetBone {boneName=thigh_twist_01_l}}
     {targetBone {boneName=thigh_twist_01_r}}
     {targetBone {boneName=thigh_l}}
    ]
    constraintWeights [1,1,1,1,1]
   }
   n26 = "Constraint - dyn_armband_knot_MSTR (PPP)"!(ComponentPose=@n25, Alpha=1, ConstraintWeights_0=0.541, ConstraintWeights_1:float=0.228, ConstraintWeights_2:float=0.231) -> (Pose) {
    alphaBoolBlend {blendOption=Linear}
    boneToModify {boneName=dyn_armband_knot_MSTR}
    constraintSetup [{targetBone {boneName=upperarm_twist_02_r}}, {targetBone {boneName=upperarm_r}}, {targetBone {boneName=upperarm_twist_02_r}}]
    constraintWeights [1,1,1]
   }
   n27 = "Constraint - master_collar_back_l (PPPPPPPP)"!(
    ComponentPose=@n26
    Alpha=0.5
    ConstraintWeights_0=0.3022
    ConstraintWeights_1:float=0.0081
    ConstraintWeights_2:float=0.3981
    ConstraintWeights_3:float=0.0011
    ConstraintWeights_4:float=0.0121
    ConstraintWeights_5:float=0.2186
    ConstraintWeights_6:float=0.0542
    ConstraintWeights_7:float=0.0056
   ) -> (Pose) {
    alphaBoolBlend {blendOption=Linear}
    boneToModify {boneName=master_collar_back_l}
    constraintSetup [
     {targetBone {boneName=clavicle_l}}
     {targetBone {boneName=dyn_collar_back_r}}
     {targetBone {boneName=dyn_collar_back_l}}
     {targetBone {boneName=spine_03}}
     {targetBone {boneName=spine_04}}
     {targetBone {boneName=spine_05}}
     {targetBone {boneName=upperarm_twist_01_l}}
     {targetBone {boneName=clavicle_r}}
    ]
    constraintWeights [1,1,1,1,1,1,1,1]
   }
   n28 = "Constraint - master_collar_back_r (PPPPPPPP)"!(
    ComponentPose=@n27
    Alpha=0.5
    ConstraintWeights_0=0.0056
    ConstraintWeights_1:float=0.056
    ConstraintWeights_2:float=0.3996
    ConstraintWeights_3:float=0.0075
    ConstraintWeights_4:float=0.0012
    ConstraintWeights_5:float=0.0126
    ConstraintWeights_6:float=0.213
    ConstraintWeights_7:float=0.3046
   ) -> (Pose) {
    alphaBoolBlend {blendOption=Linear}
    boneToModify {boneName=master_collar_back_r}
    constraintSetup [
     {targetBone {boneName=clavicle_l}}
     {targetBone {boneName=upperarm_twist_01_r}}
     {targetBone {boneName=dyn_collar_back_r}}
     {targetBone {boneName=dyn_collar_back_l}}
     {targetBone {boneName=spine_03}}
     {targetBone {boneName=spine_04}}
     {targetBone {boneName=spine_05}}
     {targetBone {boneName=clavicle_r}}
    ]
    constraintWeights [1,1,1,1,1,1,1,1]
   }
   n29 = RigidBody(ComponentPose=@n28, Alpha=SelectFloat(A=0, B=1, bPickA=Get isBackStatic()), ExternalForce=Get ext_force() -> (ext_force:Vector)) {
    bClampLinearTranslationLimitToRefPose=true
    bOverrideWorldGravity=true
    overridePhysicsAsset=&/F_MED_Cyber_Karate_Collar_Physics
    simulationSpace=BaseBoneSpace
    alphaBoolBlend {blendOption=Linear}
    alphaScaleBiasClamp {bInterpResult=true}
    baseBoneRef {boneName=spine_02}
    componentAppliedLinearAccClamp=<3000 3000 3000>
    componentLinearAccScale=<0.4 0.3 0.3>
    componentLinearVelScale=<0.4 0.3 0.3>
    overrideWorldGravity {z=-980}
   }
   n30 = RigidBody(ComponentPose=@n29) {
    bClampLinearTranslationLimitToRefPose=true
    bOverrideWorldGravity=true
    overridePhysicsAsset=&/F_MED_Cyber_Karate_Shirt_PhysicS
    simulationSpace=BaseBoneSpace
    alphaBoolBlend {blendOption=Linear}
    baseBoneRef {boneName=spine_02}
    componentAppliedLinearAccClamp=<6000 6000 6000>
    componentLinearAccScale=<0.5 0.5 0.5>
    componentLinearVelScale=<0.5 0.5 0.5>
    overrideWorldGravity {z=-980}
   }
   Save cached pose 'PreCrouch'(Pose=Component To Local(ComponentPose=@n30)) {cacheName=PreCrouch}
  ]}
 EventGraph {nodes [
   Event BlueprintUpdateAnimation()
   > n1 = Sequence!(execute) -> (then_0, then_1, then_2:Exec, then_3:Exec)
   > n2 = Branch(Condition=Get bIsCrouching())
   n3 = Get PawnSpeed2D()
   > n4 = Branch(Condition="float <= float"!(A:double=MapRangeClamped(Value=@n3, InRangeA=0, InRangeB=150), B:double=0) -> (ReturnValue:bool) {is=title K2Node_CallFunction})
   > Set crouch_speed_alpha(crouch_speed_alpha:double=1) -> (Output_Get:double)
   > Set crouch_moving(crouch_moving:bool=false) -> (Output_Get:bool)
   > Set crouch_rigid_alpha(crouch_rigid_alpha:double=0.2) -> (Output_Get:double)
   > Set move_speed_alpha(move_speed_alpha:double=0) -> (Output_Get:double)
   Set crouch_speed_alpha(execute=@n4.else, crouch_speed_alpha=0)
   > Set crouch_moving(crouch_moving=true)
   > Set crouch_rigid_alpha(crouch_rigid_alpha=0.55)
   > Set move_speed_alpha(move_speed_alpha=1)
   Set crouch_rigid_alpha(execute=@n2.else, crouch_rigid_alpha=0.8)
   Reroute Node(InputPin=@n1.then_1)
   n5 = Get bIsSkydiveFloating()
   n6 = Get bIsSkyDiving()
   n7 = Get bIsSkydiveDiveMode()
   Set Skydive_Alpha(execute=@n1.then_3, Skydive_Alpha:double=SelectFloat(A=1, B=0, bPickA=OR Boolean!(A=@n5, B=@n6, C:bool=@n7) -> (ReturnValue))) -> (Output_Get:double)
   > n8 = Branch(Condition=Get bIsJumping())
   > Delay(Duration=0.05)
   > Set ext_force(ext_force:Vector="0,0,1250") -> (Output_Get:Vector)
   > n9 = Delay(Duration=0.6)
   Do Once(execute=@n8.else, Reset=@n9.then)
   > Set ext_force(ext_force="0, 0, 0")
   n10 = Get bRigidBodyEnabled()
   n11 = "integer > integer"(A=Get CurrentLOD(), B=1) {is=title K2Node_CallFunction}
   n12 = Set isRigidBodyDisabled(execute=@n1.then_2, isRigidBodyDisabled:bool=OR Boolean!(A=NOT Boolean(A=@n10), B=@n11) -> (ReturnValue)) -> (Output_Get:bool)
   n13 = Get bIsJumping()
   n14 = Get bIsSprinting()
   > Set isBackStatic(isBackStatic:bool=OR Boolean!(A=Get isBackpackEquipped() -> (isBackpackEquipped:bool), B=@n12, C:bool=AND Boolean(A=@n13, B=@n14)) -> (ReturnValue)) -> (Output_Get:bool)
   > Set back_collar_lod_offset(back_collar_lod_offset:Vector=SelectVector(A="0,-5, 0", B="0, 0, 0", bPickA=@n11)) -> (Output_Get:Vector)
   > Set back_collar_lod_rot_offset(back_collar_lod_rot_offset:Rotator=SelectRotator!(A:Rotator="0,20, 0", B:Rotator="0, 0, 0", bPickA:bool=@n11) -> (ReturnValue:Rotator)) -> (Output_Get:Rotator)
   Event BlueprintInitializeAnimation()
   n15 = TryGetPawnOwner()
   > Is Valid(InputObject=@n15)
   > n16 = Cast To FortPlayerPawnAthena(Object=@n15)
   > Set player_pawn(player_pawn:Fort Player Pawn Athena Object Reference=@n16) -> (Output_Get:Fort Player Pawn Athena Object Reference)
   Event OnCharacterCustomizationLoaded()
   n17 = Get player_pawn() -> (player_pawn:Fort Player Pawn Athena Object Reference)
   > Is Valid(InputObject=@n17)
   > Do Once()
   n18 = CheckForBackbling()
   > Set isBackpackEquipped(isBackpackEquipped:bool=OR Boolean!(A=@n18, B=@n18.bIsBackblingLong) -> (ReturnValue)) -> (Output_Get:bool)
   Event OnBeginCrouching()
   > Set is_crouching(is_crouching:bool=true) -> (Output_Get:bool)
   Event OnEndCrouching()
   > Set is_crouching(is_crouching=false)
   "Not Equal (Object)"!(A:Object Reference=GetSkeletalMeshForPartType(self=@n17, PartType=Backpack), B:Object Reference) -> (ReturnValue:bool)
  ]}
]
variables [
 back_collar_lod_offset {type=Vector}
 back_collar_lod_rot_offset {type=Rotator}
 crouch_moving {type=boolean}
 crouch_rigid_alpha {type=number}
 crouch_speed_alpha {type=number}
 ext_force {type=Vector}
 is_crouching {type=boolean}
 isBackpackEquipped {type=boolean}
 isBackStatic {type=boolean}
 isRigidBodyDisabled {type=boolean}
 move_speed_alpha {type=number}
 player_pawn {type=FortPlayerPawnAthena}
 Skydive_Alpha {type=number}
]
