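# Animation Blueprint that adds secondary motion to the hair and face accessory of the female Cyber Karate player cosmetic.
# It copies the pose from the attached parent mesh, then runs Anim Dynamics on the bun strand bone with angular limits and heavy damping.
# A RigidBody node then simulates the FaceAcc physics asset relative to neck_02, dropping to half blend when PawnSpeed2D reaches 420 or more.
# The event graph holds only unconnected stub nodes.
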AnimBlueprint /BRCosmetics/Characters/Player/Female/Medium/Bodies/F_MED_Cyber_Karate/Meshes/Parts/F_MED_Cyber_Karate_FaceAcc_Hair_AnimBp
target_skeleton=&/F_MED_Cyber_Karate_FaceAcc_Hair_Skeleton
graphs [
 AnimGraph {nodes [
   n1 = Anim Dynamics - Bone: dyn_bun_strand(ComponentPose=Local To Component(LocalPose=Copy Pose From Mesh() {bUseAttachedParent=true}), Alpha=1, ExternalForce="0,0,0") {
    angularDampingOverride=0.9999
    bGravityOverrideInSimSpace=true
    bOverrideAngularDamping=true
    bOverrideLinearDamping=true
    gravityScale=0.75
    linearDampingOverride=0.9999
    alphaBoolBlend {blendOption=Linear}
    boundBone {boneName=dyn_bun_strand}
    componentAppliedLinearAccClamp=<400 400 400>
    componentLinearAccScale=<0.5 0.5 0.25>
    componentLinearVelScale=<0.5 0.5 0.25>
    physicsBodyDefinitions [{
      sphereCollisionRadius=0
      boundBone {boneName=dyn_bun_strand}
      boxExtents=<1 3 3>
      constraintSetup {angularLimitsMax=<0 30 10>, angularLimitsMin=<0 -30 -10>, angularTarget=<0 0 0>}
      localJointOffset=<1.5 0 0>
     }]
    relativeSpaceBone {boneName=spine_05}
   }
   n2 = Get PawnSpeed2D()
   n3 = RigidBody(ComponentPose=@n1, Alpha=SelectFloat(A=0.5, B=1, bPickA="float >= float"(A=@n2, B=420) {is=title K2Node_CallFunction})) {
    bClampLinearTranslationLimitToRefPose=true
    bOverrideWorldGravity=true
    overridePhysicsAsset=&/F_MED_Cyber_Karate_FaceAcc_Physics
    simulationSpace=BaseBoneSpace
    alphaBoolBlend {blendOption=Linear}
    baseBoneRef {boneName=neck_02}
    componentAppliedLinearAccClamp=<4000 4000 4000>
    componentLinearAccScale=<0.25 0.25 0.25>
    componentLinearVelScale=<0.25 0.25 0.25>
    overrideWorldGravity {z=-980}
   }
   Output Pose(Result=Component To Local(ComponentPose=@n3)) {layerGroup=None}
  ]}
 EventGraph {nodes [
   TryGetPawnOwner()
   Event BlueprintUpdateAnimation()
  ]}
]
variables []
